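# PoorManSmoothStep is a scalar material function in the damage text Lightning UI folder that approximates a smoothstep on a distance against a threshold.
# It subtracts threshold from distance, scales by the reciprocal of threshold and clamps the result.
# A fixed gain and second clamp turn this into a steep linear ramp just past the threshold, with no curved easing.
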MaterialFunction /Game/UI/DamageText/Lightning/PoorManSmoothStep
graph {nodes [
  n1 = FunctionInput() {inputName=distance, inputType=FunctionInput_Scalar}
  n2 = FunctionInput() {inputName=threshold, inputType=FunctionInput_Scalar}
  n3 = Clamp(None=Multiply(A=Subtract(A=@n1, B=@n2), B=Divide(B=@n2)))
  FunctionOutput(None=Clamp(None=Multiply(A=@n3) {constB=10}))
 ]}
