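# Masked unlit material for lightning-styled damage numbers, with extra flicker on critical hits.
# UV2 selects between digit and +/- textures, whose alpha is thresholded like a distance field to form the shape, border and color gradient.
# Scrolling painted lightning, distorted by that alpha, is modulated by noise and camera-distance falloff, then contrast-boosted and tinted.
# The shared DamageNumberWPO function takes blended normals and supplies animation time, letter highlight and crit flag for fading and flicker.
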Material /Game/UI/DamageText/Lightning/LightningBaseMat
graph {
 properties {
  bEnableResponsiveAA=true
  blendMode=BLEND_Masked
  numCustomizedUVs=6
  pixelDepthOffsetMode=PDOM_Legacy
  shadingModel=MSM_Unlit
  translucencyLightingMode=TLM_Surface
  usedShadingModels=Unlit
 }
 nodes [
  n1 = TextureCoordinate() {coordinateIndex=1}
  n2 = Constant()
  n3 = TextureSample(UVs=Add(A=@n1, B=@n2)) {samplerType=SAMPLERTYPE_LinearColor, texture=&/Lightning}
  n4 = Add(A=TextureCoordinate() {vTiling=0.5}, B=AppendVector(A=Constant(), B=ScalarParameter() {parameterName=+Or-}))
  n5 = TextureSample(UVs=Add(A=@n2, B=@n4)) {texture=&/BaseLightning+-.BaseLightning+-}
  n6 = ComponentMask(None=TextureCoordinate() {coordinateIndex=2}) {r=true}
  n7 = MaterialFunctionCall(Float4=LinearInterpolate(A=AppendVector(A=@n3.RGB, B=@n3.RGB), B=AppendVector(A=@n5.RGB, B=@n5.RGB), Alpha=@n6)) {materialFunction=&BreakOutFloat4Components}
  n8 = MaterialFunctionCall!(threshold=Constant() {r=0.4754}, distance=@n7.A) {materialFunction=&/PoorManSmoothStep}
  Add(
   A=Add(A=@n8.Result, B=Multiply(A=@n7.G) {constB=0.5})
   B=MaterialFunctionCall(threshold=Constant() {r=0.3}, distance=@n7.A) {desc=smaller, materialFunction=&PoorManSmoothStep}.Result
  )
  Constant3Vector()
  MaterialFunctionCall(Speed=Constant() {r=0.5}) {materialFunction=&TimeWithSpeedVariable}
  n9 = Constant() {r=1}
  n10 = TextureCoordinate() {uTiling=0.4, vTiling=0.4333}
  n11 = MaterialFunctionCall(XY=@n10) {materialFunction=&Swizzle}
  n12 = TextureCoordinate() {coordinateIndex=5}
  n13 = Add(A=Add(A=Add(A=@n11.YX, B=@n12), B=TextureCoordinate() {coordinateIndex=4}), B=Multiply(A=Time()) {constB=0.1})
  n14 = Multiply(A=@n7.A) {constB=0.02, desc=glow falloff}
  n15 = TextureSample!(
   UVs=Frac(None=Add(A=@n13, B=@n14))
   Tex
   Level=MaterialFunctionCall!(UVs=@n10, Texture Size=Constant2Vector() {g=1024, r=2048}) {materialFunction=/Engine/Functions/Engine_MaterialFunctions01/Texturing/ComputeMipLevel}.MipLevel
   Apply View MipBias
  ) {
   mipValueMode=TMVM_MipLevel
   samplerType=SAMPLERTYPE_LinearColor
   texture=/Game/Effects/Fort_Effects/World/Enviroment/Lightning/lightningPainting
  }
  n16 = CameraPositionWS()
  n17 = WorldPosition()
  n18 = Divide(
   A=Subtract(A=Distance(A=@n16, B=@n17.XYZ), B=ScalarParameter() {defaultValue=1024, parameterName=Lightning falloff start}) {constB=128}
   B=ScalarParameter() {defaultValue=1024, parameterName=Lightning falloff distance}
  ) {constB=128}
  n19 = Time()
  n20 = ScalarParameter() {defaultValue=0.2584, parameterName=AnimationSpeed}
  n21 = MaterialFunctionCall(Float2=TextureCoordinate() {coordinateIndex=4}) {materialFunction=&BreakOutFloat2Components}
  n22 = TextureSample!(UVs=Add(A=Multiply(A=@n19, B=@n20) {constB=0.1}, B=@n21.R), Tex, Apply View MipBias) {
   samplerSource=SSM_Wrap_WorldGroupSettings
   samplerType=SAMPLERTYPE_LinearColor
   texture=/Game/Packages/Fortress_Sky/Materials/Noise32x322_nearestNeighboor
  }
  n23 = Multiply(A=OneMinus(None=Clamp(None=@n18)), B=@n22.G)
  n24 = Multiply(A=Multiply(A=@n15.R) {constB=1.5}, B=@n23)
  n25 = Constant() {r=0.5708}
  n26 = Multiply(
   A=@n8.Result
   B=OneMinus(None=MaterialFunctionCall(threshold=@n25, distance=@n7.A) {desc=smaller, materialFunction=&PoorManSmoothStep}.Result)
  )
  n27 = Constant() {r=100}
  n28 = Add(
   A=MaterialFunctionCall(In=Multiply(A=@n24, B=@n8.Result)) {desc=pull in lightning, materialFunction=&CheapContrast}.Result
   B=Multiply(
    A=MaterialFunctionCall(In=@n26, Contrast=@n27) {materialFunction=&CheapContrast}.Result
    B=LinearInterpolate(Alpha=@n23) {constA=0.5}
   )
  )
  n29 = Constant() {r=0.5}
  n30 = Add(A=OneMinus(None=@n26)) {constB=0.2677}
  n31 = Multiply(
   A=MaterialFunctionCall(In=Clamp(None=@n28), Contrast=Multiply(A=@n29, B=@n30) {desc=diffuse on border}) {desc=pull in lightning, materialFunction=&CheapContrast}.Result
   B=VectorParameter() {parameterName=Lightning color, defaultValue=<rgba 1.5 1.2261 0.6577 1>}.RGB
  )
  n32 = VectorParameter() {parameterName=Center number color, defaultValue=<rgba 0.95 0.8955 0.4908 10>}
  n33 = LinearInterpolate(A=VectorParameter() {parameterName=outer number color, defaultValue=<rgba 0.85 0.7072 0.1353 1>}.RGB, B=@n32.RGB, Alpha=Power(Base=@n7.A))
  n34 = Add(
   A=Add(A=Power(Base=@n31) {constExponent=3.2504}, B=Power(Base=@n33, Exp=@n32.A))
   B=VectorParameter() {parameterName=Base number color, defaultValue=<rgba 0.2 0.1554 0.008 2>}.RGB
  )
  n35 = LinearInterpolate(
   A=TextureSampleParameter2D(UVs=@n1) {group=Visuals, parameterName=Normal, samplerType=SAMPLERTYPE_Normal, texture=/Game/UI/DamageText/Fire/FireNormalsMap}.RGB
   B=TextureSample(UVs=@n4) {samplerType=SAMPLERTYPE_Normal, texture=/Game/UI/DamageText/Fire/Fire+-Normal.Fire+-Normal}.RGB
   Alpha=@n6
  )
  n36 = Constant3Vector() {constant=<rgba 1 1 0.5 0>}
  n37 = MaterialFunctionCall!(
   Normals=Normalize(VectorInput=Multiply(A=@n35, B=@n36))
   Debug texture coordinate input
   Animation
   Additonal Scale=Constant() {r=1.2}
   "Additional Camera Offset "
   "Additonal Screenspace XY offset "
  ) {materialFunction=/Game/UI/DamageText/Base/DamageNumberWPO}
  n38 = Multiply(A=Multiply(A=Multiply(A=@n34, B=@n8.Result)) {constB=3}, B=Multiply(A=OneMinus(None=@n37.Time), B=@n37.Pixel shader Letter Highlight - uv 3))
  Multiply(
   A=MaterialFunctionCall(Color_Morning=@n9, Color_Day=@n9, Color_Evening=@n9, Color_Night=Constant() {r=0.615}) {materialFunction=&Color_Time_of_Day}.Result
   B=@n38
  )
  n39 = MaterialFunctionCall(Float2=TextureCoordinate() {coordinateIndex=4}) {materialFunction=&BreakOutFloat2Components}
  n40 = TextureCoordinate() {uTiling=0, vTiling=0.005}
  n41 = Constant() {r=0.5}
  n42 = TextureSample(UVs=Add(
    A=Add(A=@n39.R, B=@n40)
    B=MaterialFunctionCall!(TimeSpeed=@n41, Jump Amount Per Step, Number of Steps Per Unit of Time, Time Offset) {materialFunction=/Engine/Functions/Engine_MaterialFunctions02/SteppingPannerTime}.Result
   )) {samplerType=SAMPLERTYPE_Normal}
  Multiply(A=Subtract(A=ComponentMask(None=@n42.RGB) {g=true, r=true}) {constB=0.5}, B=Constant2Vector() {r=0.1385}) {constB=0.01}
  PixelDepth()
  n43 = Multiply(A=@n37.isCriticalHit?) {constB=3}
  n44 = MaterialFunctionCall(Speed=Constant() {r=5}) {materialFunction=&TimeWithSpeedVariable}
  n45 = Multiply(A=@n21.R) {constB=0.03}
  n46 = ConstantBiasScale(None=TextureSample(UVs=Add(A=@n44.Time, B=@n45)) {samplerType=SAMPLERTYPE_LinearColor, texture=&Noise32x322_nearestNeighboor}.B) {bias=-0.5, scale=2}
  n47 = Clamp(None=Multiply(A=Subtract(A=@n37.Pixel shader Letter Highlight - uv 3 0 to 1) {constB=0.8}) {constB=10})
  n48 = LinearInterpolate(B=Clamp(None=LinearInterpolate(B=@n46, Alpha=@n47) {constA=1}), Alpha=@n37.isCriticalHit?) {constA=1}
  Reroute(None=Multiply(A=Add(A=@n43) {constB=0.5}, B=Multiply(A=@n38, B=@n48)))
  ScalarParameter() {defaultValue=0.0712}
  TextureSample(UVs=@n4) {texture=/Game/UI/DamageText/Fire/FireNumbers_opacity_g_distancefield_+-.FireNumbers_opacity_g_distancefield_+-}
  TextureSample(UVs=@n4) {
   desc=DamageNumbers_Ice_RmistCore_G_Edges_B_AO
   texture=/Game/UI/DamageText/Fire/DamageNumbers_Fire_R_brightEdges_G_Crevasses_B_AO_+-.DamageNumbers_Fire_R_brightEdges_G_Crevasses_B_AO_+-
  }
 ]
}
